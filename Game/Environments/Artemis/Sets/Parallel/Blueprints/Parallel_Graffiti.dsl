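; Parallel_Graffiti is a Blueprint for configurable graffiti set dressing in the Artemis Parallel environment, displayed on a static mesh.
; Its construction script creates a dynamic material instance from the graffiti material variable on material slot 0.
; It feeds Blueprint variables into that instance for flipbook set, graffiti texture, background decal texture and extra scale.
; The event graph holds only disabled stubs, so the Blueprint adds no runtime logic.
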(event EventBeginPlay
  (|CallParentFunction) ; This node is disabled and will not be called.\nDrag off pins to build functionality.
)

(event Collision|EventActorBeginOverlap
  (params (OtherActor :type Actor%20Object%20Reference))
  (|CallParentFunction OtherActor) ; This node is disabled and will not be called.\nDrag off pins to build functionality.
)

(event EventTick
  (params (DeltaSeconds :type Float%20%28single-precision%29))
  (|CallParentFunction DeltaSeconds) ; This node is disabled and will not be called.\nDrag off pins to build functionality.
)

(fn ConstructionScript
  (|CallParentFunction)
  (bind _returnvalue (Rendering|Material|CreateDynamicMaterialInstance (Variables|Debug|GetStaticMeshComponent) 0 (Variables|Graffiti|GetGraffitiMaterial) "None" :!class /Script/Engine.PrimitiveComponent))
  (Rendering|Material|SetScalarParameterValue _returnvalue "Graffiti_FlipBookSet" (Variables|Graffiti|GetGraffiti_FlipBookSet) :!class /Script/Engine.MaterialInstanceDynamic)
  (Rendering|Material|SetTextureParameterValue _returnvalue "Graffiti" (Variables|Graffiti|GetGraffiti))
  (Rendering|Material|SetScalarParameterValue _returnvalue "GraffitiExtraScale" (Variables|Graffiti|GetBackgroundGraffitiScale) :!class /Script/Engine.MaterialInstanceDynamic)
  (Rendering|Material|SetTextureParameterValue _returnvalue "Decal" (Variables|Graffiti|GetBackground%20Graffiti%20%28Decal%29)))
