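# Niagara system for steam rising from a manhole, built from one Steam sprite emitter that spawns large, faint, flipbook-animated puffs at a steady low rate.
# Random upward-drifting acceleration and drag move the puffs, while a falloff point force fed by player position and velocity makes them react to the player.
# Fortnite time-of-day parameters tint the steam and scale its fade-in/out alpha, and sprites grow over life and with speed.
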NiagaraSystem /Game/Environments/Artemis/Sets/Parallel/FX/NS_Asteria_Manhole_Steam
dependencies {
 usedDataInterfaces [&NiagaraDataInterfaceColorCurve, &NiagaraDataInterfaceCurve]
 usedDynamicInputs [&self:ScratchDynamicInput]
 usedModules [
  &self:SystemSpawnScript.SystemScriptSource.SystemScriptGraph.NiagaraNodeAssignment_3.SetVariables_#1
  &self:SystemSpawnScript.SystemScriptSource.SystemScriptGraph.NiagaraNodeAssignment_5.SetVariables_#2
  &SystemState
  &self:SystemSpawnScript.SystemScriptSource.SystemScriptGraph.NiagaraNodeAssignment_0.SetVariables_#3
  &self:SystemSpawnScript.SystemScriptSource.SystemScriptGraph.NiagaraNodeAssignment_1.SetVariables_#4
 ]
}
emitters {Steam {
  emitterUpdateScript [
   &EmitterState()
   &SpawnRate(SpawnRate=3, Use Spawn Probability=false)
  ]
  particleSpawnScript [
   &InitializeParticle(
    Color Maximum=<rgba 1 1 1 0.15>
    Color Minimum=<rgba 1 1 1 0.1>
    Color Mode=::NewEnumerator2
    Lifetime Max=4
    Lifetime Min=3.5
    Lifetime Mode=::NewEnumerator1
    Sprite Rotation Mode=::NewEnumerator2
    Sprite Size Mode=::NewEnumerator1
    Uniform Sprite Size Max=160
    Uniform Sprite Size Min=140
    UsePositionOffset=false
    Write Material Random=true
   )
   &ApplyInitialForces(Apply Rotational Force to Rotational Velocity=false)
   &AddVelocity(Coordinate Space=::World, Velocity=&RandomRangeVector(Minimum=<10 1 0>, Maximum=<0 -1 0>)) {enabled=false}
  ]
  particleUpdateScript [
   &ParticleState()
   &CameraOffset(Camera Offset Amount=50)
   &AccelerationForce(Acceleration=&RandomRangeVector(Minimum=<30 -10 20>, Maximum=<60 10 120>))
   &PointForce(
    Force Falloff Distance=90
    Force Origin=$System.PlayerPosition
    Force Strength=350
    Linear Force Vector:Vector3f=$System.PlayerVelocity
    Module.Use Falloff=true
    Module.UseLinearForce=true
   )
   &Drag(Drag=1.5, Use Linear Drag=true, Use Rotational Drag=false)
   &SolveForcesAndVelocity(Clamp Velocity=true, Limit Acceleration=false, Speed Limit=100)
   &ScaleColor(
    Scale Alpha=&FloatFromCurve(
     FloatCurve=DataInterface() {propertyValues {Curve {keys [
         {interpMode=RCIM_Cubic, tangentMode=RCTM_User, time=0, value=0}
         <at 0.2 0.9847 RCIM_Cubic>
         {arriveTangent=-0.0405, interpMode=RCIM_Cubic, leaveTangent=-0.0405, tangentMode=RCTM_User, time=1, value=0}
        ]}}}
     Scale Curve=&Multiply_Float(
      A=0.5
      B=&FNTimeOfDayScalar(Night Scale=0.1, Morning Scale=0.1, Day Scale=0.1, Evening Scale=0.25)
     )
    )
    Scale RGB=&MakeVectorFromLinearColorRGB(Color=&ColorFromCurve(
      ColorCurve=DataInterface() {propertyValues {
        AlphaCurve {keys [<at 0 1>,<at 1 1>]}
        BlueCurve {keys [
          {arriveTangent=0.256, arriveTangentWeight=0.0864, interpMode=RCIM_Cubic, leaveTangentWeight=0.0864, time=0.001, value=0.2157}
          <at 0.1097 0.2157>
          {
           arriveTangent=-0.0692
           arriveTangentWeight=0.0231
           interpMode=RCIM_Cubic
           leaveTangent=-0.0692
           leaveTangentWeight=0.0231
           time=0.1401
           value=0.2452
          }
          {
           arriveTangent=-2.1444
           arriveTangentWeight=0.0556
           interpMode=RCIM_Cubic
           leaveTangent=-2.1444
           leaveTangentWeight=0.0556
           time=0.171
           value=0.2039
          }
          {arriveTangent=0.7909, interpMode=RCIM_Cubic, leaveTangent=0.7909, time=0.216, value=0.0824}
          {interpMode=RCIM_Cubic, tangentMode=RCTM_User, time=0.29, value=0.298}
          {
           arriveTangent=1.194
           arriveTangentWeight=0.0789
           interpMode=RCIM_Cubic
           leaveTangent=1.194
           leaveTangentWeight=0.0789
           tangentMode=RCTM_User
           tangentWeightMode=RCTWM_WeightedBoth
           time=0.375
           value=0.3229
          }
          {
           arriveTangent=0.246
           arriveTangentWeight=0.0709
           interpMode=RCIM_Cubic
           leaveTangent=0.246
           leaveTangentWeight=0.0709
           tangentMode=RCTM_User
           tangentWeightMode=RCTWM_WeightedBoth
           time=0.5
           value=0.7
          }
          {interpMode=RCIM_Cubic, tangentMode=RCTM_User, time=0.817, value=0.4588}
          {arriveTangent=-3.5889, interpMode=RCIM_Cubic, leaveTangent=-3.5889, time=0.833, value=0.4392}
          {
           arriveTangent=-2.455
           arriveTangentWeight=0.2296
           interpMode=RCIM_Cubic
           leaveTangent=-2.455
           leaveTangentWeight=0.2105
           time=0.898
           value=0.1681
          }
          {arriveTangent=0.4796, interpMode=RCIM_Cubic, leaveTangent=0.4796, time=0.9285, value=0.2047}
          {arriveTangentWeight=0.1844, interpMode=RCIM_Cubic, leaveTangent=0.0005, leaveTangentWeight=0.1844, time=1, value=0.217}
         ]}
        GreenCurve {keys [
          {arriveTangent=0.2383, arriveTangentWeight=0.0668, interpMode=RCIM_Cubic, leaveTangentWeight=0.0668, time=0.001, value=0.1765}
          <at 0.1097 0.1765>
          {arriveTangent=0.1845, interpMode=RCIM_Cubic, leaveTangent=0.1845, time=0.1401, value=0.1522}
          {
           arriveTangent=-0.9193
           arriveTangentWeight=0.0324
           interpMode=RCIM_Cubic
           leaveTangent=-0.9193
           leaveTangentWeight=0.0324
           time=0.171
           value=0.2078
          }
          {arriveTangent=0.758, interpMode=RCIM_Cubic, leaveTangent=0.758, time=0.216, value=0.0824}
          {interpMode=RCIM_Cubic, tangentMode=RCTM_User, time=0.29, value=0.298}
          {
           arriveTangent=1.194
           arriveTangentWeight=0.0789
           interpMode=RCIM_Cubic
           leaveTangent=1.194
           leaveTangentWeight=0.0789
           tangentMode=RCTM_User
           tangentWeightMode=RCTWM_WeightedBoth
           time=0.375
           value=0.2876
          }
          {
           arriveTangent=0.0845
           arriveTangentWeight=0.0575
           interpMode=RCIM_Cubic
           leaveTangent=0.0845
           leaveTangentWeight=0.0575
           tangentMode=RCTM_User
           tangentWeightMode=RCTWM_WeightedBoth
           time=0.5
           value=0.5947
          }
          {interpMode=RCIM_Cubic, tangentMode=RCTM_User, time=0.817, value=0.3137}
          {arriveTangent=-1.5352, interpMode=RCIM_Cubic, leaveTangent=-1.5352, time=0.833, value=0.4392}
          {
           arriveTangent=-1.3022
           arriveTangentWeight=0.1327
           interpMode=RCIM_Cubic
           leaveTangent=-2.6654
           leaveTangentWeight=0.1837
           tangentMode=RCTM_Break
           tangentWeightMode=RCTWM_WeightedBoth
           time=0.898
           value=0.1894
          }
          {arriveTangent=-0.128, interpMode=RCIM_Cubic, leaveTangent=-0.128, time=0.9285, value=0.1164}
          {
           arriveTangent=0.6573
           arriveTangentWeight=0.3221
           interpMode=RCIM_Cubic
           leaveTangent=0.6573
           leaveTangentWeight=0.3221
           tangentMode=RCTM_User
           tangentWeightMode=RCTWM_WeightedBoth
           time=1
           value=0.1763
          }
         ]}
        RedCurve {keys [
          {arriveTangent=0.2622, arriveTangentWeight=0.0786, interpMode=RCIM_Cubic, leaveTangentWeight=0.0786, time=0.001, value=0.1098}
          <at 0.1097 0.1098>
          {arriveTangent=1.4072, interpMode=RCIM_Cubic, leaveTangent=1.4072, time=0.1401, value=0.4115}
          {
           arriveTangent=-3.6115
           arriveTangentWeight=0.063
           interpMode=RCIM_Cubic
           leaveTangent=-3.6115
           leaveTangentWeight=0.063
           time=0.171
           value=0.349
          }
          {arriveTangent=-0.4284, interpMode=RCIM_Cubic, leaveTangent=-0.4284, time=0.216, value=0.1373}
          {interpMode=RCIM_Cubic, tangentMode=RCTM_User, time=0.29, value=0.298}
          {
           arriveTangent=1.194
           arriveTangentWeight=0.0789
           interpMode=RCIM_Cubic
           leaveTangent=1.194
           leaveTangentWeight=0.0789
           tangentMode=RCTM_User
           tangentWeightMode=RCTWM_WeightedBoth
           time=0.375
           value=0.254
          }
          {
           arriveTangent=0.0426
           arriveTangentWeight=0.1136
           interpMode=RCIM_Cubic
           leaveTangent=0.0426
           leaveTangentWeight=0.1136
           tangentMode=RCTM_User
           tangentWeightMode=RCTWM_WeightedBoth
           time=0.5
           value=0.4193
          }
          {interpMode=RCIM_Cubic, tangentMode=RCTM_User, time=0.817, value=0.3137}
          {arriveTangent=-0.0828, interpMode=RCIM_Cubic, leaveTangent=-0.0828, tangentMode=RCTM_User, time=0.833, value=0.4392}
          {
           arriveTangent=-0.8044
           arriveTangentWeight=0.2417
           interpMode=RCIM_Cubic
           leaveTangent=-0.8044
           leaveTangentWeight=0.288
           tangentMode=RCTM_User
           time=0.898
           value=0.4271
          }
          {arriveTangent=-3.1243, interpMode=RCIM_Cubic, leaveTangent=-3.1243, time=0.9285, value=0.1053}
          {arriveTangentWeight=0.3333, interpMode=RCIM_Cubic, tangentMode=RCTM_User, tangentWeightMode=RCTWM_WeightedBoth, time=1, value=0.1085}
         ]}
       }}
      CurveIndex=Linked() {linkedVariable {name=NPC.FortniteNPC.TimeOfDay_0-1, type=NiagaraFloat}}
     ))
    ScaleA=true
    ScaleRGB=true
   )
   &ScaleSpriteSizeBySpeed(Max Scale Factor=<1.1 1.1>, Velocity Threshold=100)
   &ScaleSpriteSize(Uniform Curve Sprite Scale=DataInterface() {propertyValues {Curve {keys [{tangentMode=RCTM_User, time=0, value=0.66}, {tangentMode=RCTM_User, time=1, value=1}]}}})
   &SubUVAnimation(Start Frame:NiagaraInt32=&RandomRangeInt(Minimum=0, Maximum=10))
  ]
  rendererClasses [&NiagaraSpriteRendererProperties]
  renderers [{rendererClass=&NiagaraSpriteRendererProperties}]
 }}
summary {
 systemName=NS_Asteria_Manhole_Steam
 emitters [{emitterName=Steam, rendererClasses [&NiagaraSpriteRendererProperties]}]
 userVariables []
}
